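# Post-process blendable for Pilgrim camera tracks that fades scene color to black with distance, with alpha output enabled.
# The DepthFade parameter's R and G channels bound a reversed SmoothStep on scene depth, giving a mask of one near and zero far.
# When G does not exceed R, color passes through unchanged and the mask becomes a hard cutoff at G.
# The mask also feeds a separate reroute alongside the color result.
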Material /PilgrimCore/Cameras/M_PilgrimTrackBlendable
graph {
 properties {
  blendableOutputAlpha=true
  blendablePriority=10
  materialDomain=MD_PostProcess
  pixelDepthOffsetMode=PDOM_Legacy
  usedShadingModels=Default Lit
 }
 nodes [
  n1 = SceneTexture() {sceneTextureId=PPI_PostProcessInput0}
  n2 = VectorParameter() {parameterName=DepthFade, defaultValue=<rgba 0 0 0 1>}
  n3 = ComponentMask(None=SceneTexture() {sceneTextureId=PPI_SceneDepth}.Color) {r=true}
  n4 = Constant() {r=1}
  n5 = If(A=@n2.G, B=@n3, "A > B"=@n4, "A == B"=@n4, "A < B"=Constant())
  n6 = If(A=@n2.G, B=@n2.R, "A > B"=SmoothStep(Min=@n2.G, Max=@n2.R, Value=@n3), "A == B"=@n5, "A < B"=@n5)
  If(A=@n2.G, B=@n2.R, "A > B"=Multiply(A=@n1.Color, B=@n6), "A == B"=@n1.Color, "A < B"=@n1.Color)
  Reroute(None=@n6)
 ]
}
