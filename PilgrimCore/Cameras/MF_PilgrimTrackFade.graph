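# Material function that outputs a Fade mask from UVs, clearing a rectangular region and fading near the UV borders.
# Scaled and offset UVs are tested against InnerFadeBounds and OuterFadeBounds, using SmoothStep where the bounds separate and hard steps otherwise.
# The squared, inverted edge terms combine into a mask that is zero inside and nearly one outside.
# This multiplies a linear border falloff from FadeLength, guarded against zero-length division.
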MaterialFunction /PilgrimCore/Cameras/MF_PilgrimTrackFade
graph {nodes [
  n1 = VectorParameter() {parameterName=InnerFadeBounds, defaultValue=<rgba 0 0 1 0.4167>}
  n2 = AppendVector(A=@n1.R, B=@n1.R)
  n3 = VectorParameter() {parameterName=OuterFadeBounds, defaultValue=<rgba 0 0 1 0.4394>}
  n4 = AppendVector(A=@n3.R, B=@n3.R)
  n5 = VectorParameter() {parameterName=UVScaleOffset, defaultValue=<rgba 1 1 0 0>}
  n6 = AppendVector(A=@n5.R, B=@n5.R)
  n7 = FunctionInput(Preview=TextureCoordinate()) {bUsePreviewValueAsDefault=true, inputName=UV, inputType=FunctionInput_Vector2}
  n8 = Multiply(A=@n6) {constB=-0.5}
  n9 = Add(A=Multiply(A=@n7, B=@n6), B=Add(A=Add(A=@n8) {constB=0.5}, B=AppendVector(A=@n5.B, B=@n5.B)))
  n10 = Constant() {r=1}
  n11 = If(A=@n2, B=@n9, "A > B"=@n10, "A == B"=@n10, "A < B"=Constant())
  n12 = AppendVector(A=@n1.B, B=@n1.B)
  n13 = AppendVector(A=@n3.B, B=@n3.B)
  n14 = Constant() {r=1}
  n15 = If(A=@n12, B=@n9, "A > B"=Constant(), "A == B"=@n14, "A < B"=@n14)
  n16 = AppendVector(
   A=If(A=@n2, B=@n4, "A > B"=SmoothStep(Min=@n2, Max=@n4, Value=@n9) {constMax=0.25, constValue=0.275}, "A == B"=@n11, "A < B"=@n11)
   B=If(A=@n12, B=@n13, "A > B"=@n15, "A == B"=@n15, "A < B"=SmoothStep(Min=@n12, Max=@n13, Value=@n9) {constMax=0.25, constValue=0.275})
  )
  n17 = MaterialFunctionCall(Float4=OneMinus(None=Power(Base=@n16))) {materialFunction=&BreakOutFloat4Components}
  n18 = OneMinus(None=Multiply(A=Multiply(A=@n17.R, B=@n17.R), B=Multiply(A=@n17.B, B=@n17.B)))
  n19 = Subtract(B=@n7) {constA=0}
  n20 = Constant4Vector() {constant=<rgba 0 0 1 1>}
  n21 = VectorParameter() {parameterName=FadeLength}
  n22 = Multiply(A=AppendVector(A=@n21.RGB, B=@n21.RGB), B=AppendVector(A=@n6, B=@n6))
  n23 = Divide(B=@n22)
  n24 = Multiply(
   A=Add(A=AppendVector(A=@n19, B=@n7), B=Subtract(A=@n22, B=@n20.RGBA))
   B=If(A=@n22, "A > B"=@n23, "A == B"=Constant4Vector().RGBA, "A < B"=@n23) {equalsThreshold=0}
  )
  n25 = MaterialFunctionCall(Float4=Saturate(None=OneMinus(None=@n24))) {materialFunction=&BreakOutFloat4Components}
  n26 = Multiply(A=@n25.R, B=@n25.R)
  n27 = Multiply(A=@n25.B, B=@n25.B)
  FunctionOutput(None=Multiply(A=Multiply(A=@n18) {constA=1, constB=0.975}, B=Multiply(A=@n26, B=@n27))) {outputName=Fade}
 ]}
